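# M_SnowPine_Silohette is a Default Lit material in the Asteria snowy pine foliage set, named as a silhouette version of the tree.
# It holds only three constant nodes: a very dark blue-green color and two scalar values, with no textures or parameters.
# The graph shows no links from these constants to material outputs.
# Automatic setting of usage flags in the editor is disabled.
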Material /Game/Environments/Asteria/Foliage/Trees/SnowyPine/Materials/M_SnowPine_Silohette
graph {
 properties {bAutomaticallySetUsageInEditor=false, usedShadingModels=Default Lit}
 nodes [
  Constant3Vector() {constant=<rgba 0.0038 0.0132 0.02 1>}
  Constant() {r=0.2}
  Constant() {r=0.8}
 ]
}
