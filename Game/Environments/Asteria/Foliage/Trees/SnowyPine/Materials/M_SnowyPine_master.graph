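# Master subsurface material for Asteria snowy pine foliage, rendering snow and ice on instanced or Nanite meshes.
# The red mask isolates ice, adding BumpOffset parallax, sky-light environment reflection and tinted tips, while mask channels also drive metallic, roughness, opacity and subsurface color.
# DestructibleDamage animation adds World Position Offset on the default shading path only, bypassing Forward and Mobile.
# A static switch routes the result through ImposterCaptureSwitch for impostor baking.
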Material /Game/Environments/Asteria/Foliage/Trees/SnowyPine/Materials/M_SnowyPine_master
graph {
 properties {
  bUseMaterialAttributes=true
  bUsedWithInstancedStaticMeshes=true
  bUsedWithNanite=true
  pixelDepthOffsetMode=PDOM_Legacy
  refractionMethod=RM_IndexOfRefraction
  shadingModel=MSM_Subsurface
  usedShadingModels=Subsurface
 }
 nodes [
  Constant()
  n1 = TextureSample() {
   samplerSource=SSM_Wrap_WorldGroupSettings
   samplerType=SAMPLERTYPE_Normal
   texture=/Game/Environments/Asteria/Foliage/Trees/SnowyPine/Textures/Snow_LP_Snow_Normal
  }
  n2 = LinearInterpolate(A=Constant3Vector() {constant=<rgba 0 0 1 1>}, B=@n1.RGB, Alpha=ScalarParameter() {defaultValue=1, parameterName=Normal Intensity})
  n3 = Transform(None=@n2)
  n4 = TextureSample() {
   samplerSource=SSM_Wrap_WorldGroupSettings
   samplerType=SAMPLERTYPE_Masks
   texture=/Game/Environments/Asteria/Foliage/Trees/SnowyPine/Textures/Snow_LP_Snow_masks
  }
  n5 = SkyLightEnvMapSample(Direction=@n3, Roughness=@n4.B)
  LinearInterpolate(A=@n2, B=@n5, Alpha=@n4.R)
  n6 = Add(A=ScalarParameter() {defaultValue=-1, parameterName=Ice Depth}, B=Multiply(A=@n4.R, B=@n4.R))
  n7 = TextureSample(UVs=LinearInterpolate(A=TextureCoordinate(), B=BumpOffset(Height=@n6), Alpha=@n4.R)) {samplerSource=SSM_Wrap_WorldGroupSettings, texture=/Game/Environments/Asteria/Foliage/Trees/SnowyPine/Textures/Snow_LP_Snow_BaseColor}
  n8 = OneMinus(None=@n4.G)
  n9 = ScalarParameter() {defaultValue=2, parameterName=IceTipsPower}
  n10 = LinearInterpolate(
   A=LinearInterpolate(A=@n7.RGB, B=Add(A=@n7.RGB, B=@n5), Alpha=@n4.R)
   B=VectorParameter() {parameterName=IceTipsTint, defaultValue=<rgba 0.9635 0.0663 1 1>}.RGB
   Alpha=Multiply(A=@n4.R, B=Power(Base=@n8, Exp=@n9) {constExponent=3})
  )
  n11 = Constant3Vector() {constant=<rgba 0.0038 0.0059 0.0552 0>}
  n12 = Power(Base=Multiply(A=ComponentMask(None=@n2) {g=true}, B=@n4.R), Exp=Constant() {r=2.5})
  n13 = Power(Base=@n4.G, Exp=ScalarParameter() {defaultValue=1.5, parameterName=Ice Metallic Power})
  n14 = Multiply(A=Power(Base=@n4.G, Exp=ScalarParameter() {parameterName=Subsurface Power}), B=@n4.R)
  n15 = Constant() {r=0.15}
  n16 = OneMinus(None=@n4.R)
  n17 = SetMaterialAttributes!(
   MaterialAttributes
   Base Color=Multiply(A=ScalarParameter() {defaultValue=1, parameterName=Ice Color Brightness}, B=LinearInterpolate(A=@n10, B=@n11, Alpha=@n12))
   Metallic=LinearInterpolate(A=Constant(), B=ScalarParameter() {defaultValue=0.5, parameterName=Metallic Ice}, Alpha=Multiply(A=@n4.R, B=@n13))
   Specular=Constant() {r=1}
   Roughness=@n4.B
   Anisotropy
   Emissive Color
   Opacity=LinearInterpolate(A=ScalarParameter() {defaultValue=1, parameterName=Opacity Base}, B=ScalarParameter() {parameterName=Opacity Edge}, Alpha=@n14)
   Opacity Mask
   Normal=@n2
   Tangent
   World Position Offset
   Subsurface Color=LinearInterpolate(
    A=VectorParameter() {parameterName=Subsurface Base}.RGB
    B=VectorParameter() {parameterName=Subsurface Edge, defaultValue=<rgba 1 0 0 0>}.RGB
    Alpha=LinearInterpolate(A=@n14, B=@n15, Alpha=@n16)
   )
   Custom Data 0
   Custom Data 1
  ) {attributeSetTypes [
    #1
    #2
    #3
    #4
    #5
    #6
    #7
    #8
    #9
    #10
    #11
    #12
    #13
    #14
   ]}
  n18 = GetMaterialAttributes(None=@n17) {attributeGetTypes [#11]}
  n19 = MaterialFunctionCall() {materialFunction=&MF_FN_DestructibleDamage}
  n20 = ShadingPathSwitch(
   Default=SetMaterialAttributes!(MaterialAttributes=@n18.MaterialAttributes, World Position Offset=Add(A=@n18.World Position Offset, B=@n19.DamageAnimWPO)) {attributeSetTypes [#11]}
   Forward=@n18.MaterialAttributes
   Mobile=@n18.MaterialAttributes
  )
  StaticSwitch(
   True=MaterialFunctionCall!(Material Attributes=@n20, IsTangentSpaceNormal=StaticBool() {desc=Don't change, value=true}, In=Constant() {r=1}) {materialFunction=/Game/ImpostorBaker/MaterialFunctions/ImposterCaptureSwitch}.Result
   False=@n20
   Value=NamedRerouteDeclaration(None=StaticBoolParameter() {parameterName=EnableForImpostorBake}) {name=IsImpostorBake, nodeColor=<rgba 1 0.3984 0 1>}
  )
  TextureSample(UVs=ComponentMask(None=ReflectionVectorWS(CustomWorldNormal=@n3)) {g=true, r=true}) {samplerType=SAMPLERTYPE_LinearColor, texture=/Game/Environments/Asteria/Foliage/Trees/SnowyPine/Textures/winter_river_1k}
 ]
}
